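# Physics asset for the HeavenSip backpack cosmetic, giving its small limbed figure capsule bodies for physics-driven secondary motion.
# A constraint chain links MSTR_body to dyn_head, dyn_body and dyn_pelvis, so the figure attaches by its head, with arms branching from the body and legs from the pelvis.
# Every joint locks twist, with narrow swing angles along the central chain and wider ones on the limbs.
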PhysicsAsset /BRCosmetics/Accessories/FORT_Backpacks/Backpack_F_MED_HeavenSip/Meshes/F_MED_HeavenSip_Pack_Physics
bodies {
 MSTR_body [{shapeType=Capsule, length=4, radius=2}]
 dyn_arm_l [{shapeType=Capsule, length=0.7152, radius=2.6746, shapeName=dyn_arm_r_capsule, center=<2.5187 0 -1.1232>, rotation=<rot 90 0 0>}]
 dyn_arm_r [{shapeType=Capsule, length=0.7152, radius=2.6746, shapeName=dyn_arm_r_capsule, center=<2.5187 0 -1.1232>, rotation=<rot 90 0 0>}]
 dyn_body [{shapeType=Capsule, length=14.1631, radius=4.2687, shapeName=dyn_body_capsule, center=<1.2413 0 0>}]
 dyn_head [{shapeType=Capsule, length=13.9631, radius=4.2687, shapeName=dyn_head_capsule, center=<1.2413 0 0>}]
 dyn_leg_l [{shapeType=Capsule, length=0.5, radius=2, shapeName=dyn_leg_l_capsule, center=<1 0 0>, rotation=<rot 90 0 0>}]
 dyn_leg_r [{shapeType=Capsule, length=0.5, radius=2, shapeName=dyn_leg_r_capsule, center=<1 0 0>, rotation=<rot 90 0 0>}]
 dyn_pelvis [{shapeType=Capsule, length=14.6607, radius=3.6606, center=<1 0 0>}]
}
constraints [
 {bone1Name=dyn_pelvis, bone2Name=dyn_body, swing1LimitDegrees=8, swing2LimitDegrees=2, twistLimitDegrees=5, twistMotion=Locked}
 {bone1Name=dyn_body, bone2Name=dyn_head, swing1LimitDegrees=6, swing2LimitDegrees=2, twistLimitDegrees=0, twistMotion=Locked}
 {bone1Name=dyn_leg_l, bone2Name=dyn_pelvis, swing1LimitDegrees=25, swing2LimitDegrees=15, twistLimitDegrees=0, twistMotion=Locked}
 {bone1Name=dyn_leg_r, bone2Name=dyn_pelvis, swing1LimitDegrees=25, swing2LimitDegrees=15, twistLimitDegrees=0, twistMotion=Locked}
 {bone1Name=dyn_head, bone2Name=MSTR_body, swing1LimitDegrees=4, swing2LimitDegrees=1, twistLimitDegrees=0, twistMotion=Locked}
 {bone1Name=dyn_arm_l, bone2Name=dyn_body, swing1LimitDegrees=5, swing2LimitDegrees=40, twistLimitDegrees=0, twistMotion=Locked}
 {bone1Name=dyn_arm_r, bone2Name=dyn_body, swing1LimitDegrees=5, swing2LimitDegrees=40, twistLimitDegrees=0, twistMotion=Locked}
]
